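# MetaSound preset of MSS_VO_BR_Boss that plays Henchman 1's convoy full-alert voice lines as a mono one-shot.
# It supplies seven wave variations, with delay and random pitch and volume offsets zeroed and probability at full.
# It uses convoy guard attenuation and the critical enemy VO class, and sends to music, weapon and explosion sidechain buses.
# Boss combat chatter concurrency is set, alongside an override allowing one instance per owner that prevents new ones.
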MetaSoundSource /VividRazor_Minions/Audio/VO/Henchman_1/Convoy/MSP_VO_S34_Henchman_1_Convoy_FullAlert
preset_parent=/Game/Sounds/Fort_Human_VO_BR/NPCs/Shared/MSS_VO_BR_Boss
inputs [
 "Delay Max (Seconds)" {type=Time}
 "Delay Min (Seconds)" {type=Time}
 "Probability (0-1)" {type=Float}
 "Random Pitch Max (Semitones)" {type=Float}
 "Random Pitch Min (Semitones)" {type=Float}
 "Random Volume Change Max (Decibels)" {type=Float}
 "Random Volume Change Min (Decibels)" {type=Float}
 UE.Source.OnPlay {type=Trigger}
 Wave Array {type=WaveAsset:Array}
]
interfaces [UE.OutputFormat.Mono:1.0, UE.Source.OneShot:1.0, UE.Source:1.1]
outputs [
 "UE.OutputFormat.Mono.Audio:0" {type=Audio}
 UE.Source.OneShot.OnFinished {type=Trigger}
]
pages [Default {input_defaults {
   "Delay Max (Seconds)"=0
   "Delay Min (Seconds)"=0
   "Probability (0-1)"=1
   "Random Pitch Max (Semitones)"=0
   "Random Pitch Min (Semitones)"=0
   "Random Volume Change Max (Decibels)"=0
   "Random Volume Change Min (Decibels)"=0
   UE.Source.OnPlay=null
   Wave Array [
    /VividRazor_Minions/Audio/VO/Henchman_1/Convoy/Waves/VO_S34_Henchman_1_Convoy_FullAlert_010
    /VividRazor_Minions/Audio/VO/Henchman_1/Convoy/Waves/VO_S34_Henchman_1_Convoy_FullAlert_011
    /VividRazor_Minions/Audio/VO/Henchman_1/Convoy/Waves/VO_S34_Henchman_1_Convoy_FullAlert_020
    /VividRazor_Minions/Audio/VO/Henchman_1/Convoy/Waves/VO_S34_Henchman_1_Convoy_FullAlert_022
    /VividRazor_Minions/Audio/VO/Henchman_1/Convoy/Waves/VO_S34_Henchman_1_Convoy_FullAlert_041
    /VividRazor_Minions/Audio/VO/Henchman_1/Convoy/Waves/VO_S34_Henchman_1_Convoy_FullAlert_060
    /VividRazor_Minions/Audio/VO/Henchman_1/Convoy/Waves/VO_S34_Henchman_1_Convoy_FullAlert_061
   ]
  }}]
properties {
 assetImportData=&self:AssetImportData
 attenuationSettings=/VividRazor_Minions/Audio/VO/Att_NPC_VO_Convoy_Guards
 maxDistance=12500
 sampleRate=48000
 soundClassObject=&Fort_VO_Enemy_Critical
 volume=3
 busSends [
  {audioBus=&AB_Sidechain_FortMusic_Vol, sendLevel=0.25}
  {audioBus=/Game/Sounds/Modulation/EnvelopeFollowers/Volume/AB_Sidechain_FortWeapon1P_Vol, sendLevel=0.25}
  {audioBus=&AB_Sidechain_FortWeapon_Vol, sendLevel=0.25}
  {audioBus=&AB_Sidechain_FortExplosion_Vol, sendLevel=0.25}
 ]
 concurrencyOverrides {bLimitToOwner=true, maxCount=1, resolutionRule=PreventNew}
 concurrencySet [&NPC_Boss_Combat_Chatter_Conc]
}
